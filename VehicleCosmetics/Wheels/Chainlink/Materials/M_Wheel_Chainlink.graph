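# Master material for the Chainlink vehicle cosmetic wheel, built on the simple wheel parent function, with support for skeletal meshes and mesh deformers.
# It overrides World Position Offset with a masked engine wiggle, using sines and cosines of time and object-relative position scaled by a vector parameter.
# An optional, tiled matte detail normal for tire rubber is blended into the parent normal.
# The result feeds the base vehicle damage function, then the de-rez transition.
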Material /VehicleCosmetics/Wheels/Chainlink/Materials/M_Wheel_Chainlink
graph {
 properties {
  bAutomaticallySetUsageInEditor=false
  bUseMaterialAttributes=true
  bUsedWithMeshDeformer=true
  bUsedWithSkeletalMesh=true
  pixelDepthOffsetMode=PDOM_Legacy
  usedShadingModels=Default Lit
 }
 nodes [
  n1 = MaterialFunctionCall() {materialFunction=&MF_WheelParent_Simple}
  n2 = NamedRerouteUsage()
  n3 = NamedRerouteUsage()
  MaterialFunctionCall(In=MaterialFunctionCall(Material=SetMaterialAttributes!(MaterialAttributes=@n1.MaterialAttributes, World Position Offset=@n2, Normal=@n3) {attributeSetTypes [#1, #2]}) {materialFunction=&MF_BaseVehicle_Damage}.Result) {materialFunction=&MF_VehicleDeRez_Transition}
  n4 = Transform(None=Subtract(A=ObjectPositionWS(), B=WorldPosition().XYZ)) {transformSourceType=TRANSFORMSOURCE_Local}
  n5 = Add(A=Multiply(A=ComponentMask(None=@n4) {r=true}, B=Constant() {r=0.01}), B=Multiply(A=Time(), B=Constant() {r=50}))
  n6 = Sine(None=@n5)
  n7 = Add(A=@n5, B=@n6)
  n8 = AppendVector(A=Sine(None=@n7), B=Cosine(None=@n7))
  n9 = VectorParameter() {group=EngineWiggle, parameterName=WiggleAmount GreenVP, defaultValue=<rgba 0 0.05 0.05 1>}
  n10 = Multiply(
   A=Multiply(A=AppendVector(A=@n8, B=@n6), B=Transform(None=@n9.RGB) {transformSourceType=TRANSFORMSOURCE_World, transformType=TRANSFORM_Local})
   B=Constant() {r=2}
  )
  n11 = Constant()
  n12 = OneMinus(None=ComponentMask(None=NamedRerouteUsage()) {a=true})
  NamedRerouteDeclaration(None=LinearInterpolate(A=LinearInterpolate(A=@n10, B=@n11, Alpha=@n12), B=@n11, Alpha=NamedRerouteUsage())) {name=WPO, nodeColor=<rgba 1 0.1633 0.1308 1>}
  NamedRerouteDeclaration(None=@n1.RimRGBAMask) {name=RIM_M, nodeColor=<rgba 0.1953 0 1 1>}
  NamedRerouteDeclaration(None=@n1.VertexColorR) {name=VC_R, nodeColor=<rgba 1 0 0.1317 1>}
  NamedRerouteDeclaration(None=GetMaterialAttributes(None=@n1.MaterialAttributes) {attributeGetTypes [#2]}.Normal) {name=Normal_IN, nodeColor=<rgba 0.257 0.1402 1 1>}
  n13 = NamedRerouteUsage()
  n14 = TextureSample(UVs=Multiply(
    A=TextureCoordinate() {uTiling=2, vTiling=2}
    B=ScalarParameter() {defaultValue=1, group=TireRubber, parameterName=TextureDetailNormalTiling}
   )) {samplerType=SAMPLERTYPE_Normal, texture=/VehicleCosmetics/SharedTextures/T_Matte_N}
  NamedRerouteDeclaration(None=LinearInterpolate(A=@n13, B=StaticSwitchParameter(True=@n14.RGB, False=@n13) {group=TireRubber, parameterName=TireRubberExist}, Alpha=NamedRerouteUsage())) {name=Normal, nodeColor=<rgba 0.8713 0 1 1>}
 ]
}
