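# Translucent unlit particle material in the Chainlink wheel cosmetic VFX set, enabled for Niagara sprites, ribbons and mesh particles.
# A fire cloud texture's RGB is multiplied by particle color to produce a tinted color.
# Its red channel times particle alpha feeds a depth fade, softening opacity where particles meet scene geometry.
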Material /VehicleCosmetics/Wheels/Chainlink/VFX/M_ChainLink_VFX
graph {
 properties {
  bAutomaticallySetUsageInEditor=false
  bUsedWithNiagaraMeshParticles=true
  bUsedWithNiagaraRibbons=true
  bUsedWithNiagaraSprites=true
  blendMode=BLEND_Translucent
  refractionMethod=RM_PixelNormalOffset
  shadingModel=MSM_Unlit
  usedShadingModels=Unlit
 }
 nodes [
  n1 = TextureSample() {texture=&/FireCloud_02_T}
  n2 = ParticleColor()
  DepthFade(Opacity=Multiply(A=@n1.R, B=@n2.A)) {fadeDistanceDefault=50}
  Multiply(A=@n1.RGB, B=@n2.RGB)
 ]
}
