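# Niagara smoke effect for the Chainlink cosmetic wheel; the emitter names suggest a menu variant.
# Two looping GPU sprite emitters spawn short-lived dark puffs from small spheres offset along Y, launched upward toward opposite X sides, one with randomized velocity.
# Upward acceleration lifts the puffs while curves over lifetime grow their size, fade their alpha and damp their velocity.
# Dynamic material parameter modules are disabled, and no user parameters are exposed.
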NiagaraSystem /VehicleCosmetics/Wheels/Chainlink/VFX/NS_Chainlink
dependencies {
 usedDataInterfaces [&NiagaraDataInterfaceCurve, &NiagaraDataInterfaceVectorCurve]
 usedDynamicInputs []
 usedModules [&SystemState]
}
emitters {
 OutsmokeMenu {
  simTarget=GPUComputeSim
  emitterUpdateScript [
   &EmitterState(Life Cycle Mode=::NewEnumerator1, Loop Delay=0.35, Loop Duration=1, UseLoopDelay=true)
   &SpawnRate(SpawnRate=45, Use Spawn Probability=false)
  ]
  particleSpawnScript [
   &InitializeParticle(
    Color=<rgba 0 0 0 1>
    Lifetime Max=0.38
    Lifetime Min=0.28
    Lifetime Mode=::NewEnumerator1
    Sprite Rotation Angle Max=126
    Sprite Rotation Angle Min=-126
    Sprite Rotation Mode=::NewEnumerator2
    Sprite Size Mode=::NewEnumerator1
    Uniform Sprite Size Max=15
    Uniform Sprite Size Min=10
    UsePositionOffset=false
    Write Material Random=true
   )
   &AddVelocity(Velocity=&RandomRangeVector(Minimum=<40 5 90>, Maximum=<40 20 150>))
   &ShapeLocation(MaskSpawnGroup=false, Override Seed=false, Sphere Radius=4, UseOwnerScale=true)
   &OffsetPosition(Position Offset=<0 45 0>)
  ]
  particleUpdateScript [
   &ParticleState()
   &ScaleColor(
    Scale Alpha=&FloatFromCurve(FloatCurve=DataInterface() {propertyValues {Curve {keys [<at 0 0.35>,<at 0.2 0.5>,<at 1 0>]}}})
    ScaleA=true
    ScaleRGB=true
   )
   &ScaleSpriteSize(Uniform Curve Sprite Scale=DataInterface() {propertyValues {Curve {keys [<at 0 1.25>,<at 0.3 2>,<at 1 3>]}}})
   &AccelerationForce(Acceleration=&RandomRangeVector(Minimum=<0 0 200>, Maximum=<1 1 300>))
   &ScaleVelocity(Velocity Scale=&ScaleFloatByCurve(
     InitialValue:Vector3f=<1 1 1>
     ScaleCurve:NiagaraDataInterfaceVectorCurve=DataInterface() {propertyValues {
       XCurve {keys [<at 0 1.55>,<at 0.65 0.67>,<at 1 0>]}
       YCurve {keys [<at 0 1.55>,<at 0.65 0.67>,<at 1 0>]}
       ZCurve {keys [<at 0 1.55>,<at 0.65 0.67>,<at 1 0>]}
      }}
     CurveIndex:NiagaraFloat=$Particles.NormalizedAge
    ))
   &SolveForcesAndVelocity(Clamp Velocity=false, Limit Acceleration=false)
   &DynamicMaterialParameters(
    Index 0 Param 2=1
    Module.Index0Float0Write=true
    Module.Index0Float1Write=true
    Module.Index0Float2Write=true
    Module.Index0Float3Write=true
   ) {enabled=false}
  ]
  rendererClasses [&NiagaraSpriteRendererProperties]
  renderers [{rendererClass=&NiagaraSpriteRendererProperties}]
 }
 Outsmoke_Menu2 {
  simTarget=GPUComputeSim
  emitterUpdateScript [
   &EmitterState(Life Cycle Mode=::NewEnumerator1, Loop Delay=0.35, Loop Duration=1, UseLoopDelay=true)
   &SpawnRate(SpawnRate=45, Use Spawn Probability=false)
  ]
  particleSpawnScript [
   &InitializeParticle(
    Color=<rgba 0.0073 0.0073 0.0073 1>
    Lifetime Max=0.45
    Lifetime Min=0.35
    Lifetime Mode=::NewEnumerator1
    Sprite Rotation Angle Max=126
    Sprite Rotation Angle Min=-126
    Sprite Rotation Mode=::NewEnumerator2
    Sprite Size Mode=::NewEnumerator1
    Uniform Sprite Size Max=15
    Uniform Sprite Size Min=10
    UsePositionOffset=false
    Write Material Random=true
   )
   &AddVelocity(Velocity=<-40 10 90>)
   &ShapeLocation(MaskSpawnGroup=false, Override Seed=false, Sphere Radius=4, UseOwnerScale=true)
   &OffsetPosition(Position Offset=<0 45 0>)
  ]
  particleUpdateScript [
   &ParticleState()
   &ScaleColor(
    Scale Alpha=&FloatFromCurve(FloatCurve=DataInterface() {propertyValues {Curve {keys [<at 0 0.35>,<at 0.2 0.5>,<at 1 0>]}}})
    ScaleA=true
    ScaleRGB=true
   )
   &ScaleSpriteSize(Uniform Curve Sprite Scale=DataInterface() {propertyValues {Curve {keys [<at 0 1.25>,<at 0.3 2>,<at 1 3>]}}})
   &AccelerationForce(Acceleration=&RandomRangeVector(Minimum=<0 0 200>, Maximum=<1 1 300>))
   &ScaleVelocity(Velocity Scale=&ScaleFloatByCurve(
     InitialValue:Vector3f=<1 1 1>
     ScaleCurve:NiagaraDataInterfaceVectorCurve=DataInterface() {propertyValues {
       XCurve {keys [<at 0 1.55>,<at 0.65 0.67>,<at 1 0>]}
       YCurve {keys [<at 0 1.55>,<at 0.65 0.67>,<at 1 0>]}
       ZCurve {keys [<at 0 1.55>,<at 0.65 0.67>,<at 1 0>]}
      }}
     CurveIndex:NiagaraFloat=$Particles.NormalizedAge
    ))
   &SolveForcesAndVelocity(Clamp Velocity=false, Limit Acceleration=false)
   &DynamicMaterialParameters(
    Index 0 Param 2=1
    Module.Index0Float0Write=true
    Module.Index0Float1Write=true
    Module.Index0Float2Write=true
    Module.Index0Float3Write=true
   ) {enabled=false}
  ]
  rendererClasses [&NiagaraSpriteRendererProperties]
  renderers [{rendererClass=&NiagaraSpriteRendererProperties}]
 }
}
summary {
 systemName=NS_Chainlink
 emitters [
  {emitterName=OutsmokeMenu, simTarget=GPUComputeSim, rendererClasses [&NiagaraSpriteRendererProperties]}
  {emitterName=Outsmoke_Menu2, simTarget=GPUComputeSim, rendererClasses [&NiagaraSpriteRendererProperties]}
 ]
 userVariables []
}
